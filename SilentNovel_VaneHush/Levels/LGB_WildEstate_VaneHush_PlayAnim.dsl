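; A Local Gameplay Behavior for the VaneHush wild estate that loops an animation montage on the activating actor.
; On execute it stores the actor and plays the first Montage Array entry, setting an event timer to the montage's returned length.
; When the timer fires it flips MontageSwitch, which nothing here reads, and replays that same first entry.
; Abort stops the current montage but does not clear the timer.
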(event EventBeginPlay)

(event Collision|EventActorBeginOverlap
  (params (OtherActor :type Actor%20Object%20Reference)))

(event EventTick
  (params (DeltaSeconds :type Float%20%28single-precision%29)))

(event Custom|PlayNewMontage
  (CallFunction|LGBPlayMontage (Utilities|Array|Get(acopy) (Variables|Default|GetMontage%20Array))))

(event LocalGameplayBehavior|EventOnExecute
  (params (Activator :type Actor%20Object%20Reference))
  (Variables|Default|SetActivatingActor Activator)
  (CallFunction|LGBPlayMontage (Utilities|Array|Get(acopy) (Variables|Default|GetMontage%20Array))))

(event Custom|LGB_PlayMontage
  (params (Anim%20Montage :type Anim%20Montage%20Object%20Reference))
  (bind _output_get (Variables|Default|SetPlayingMontage Anim%20Montage))
  (bind _returnvalue (Animation|Utilities|PlayAnimMontage (Variables|Default|GetActivatingActor) _output_get))
  (Utilities|Time|SetTimerbyEvent (EventDispatchers|CreateEvent :!event Custom|OnMontageEnd) _returnvalue))

(event Custom|OnMontageEnd
  (Variables|Default|SetMontageSwitch (not (Variables|Default|GetMontageSwitch)))
  (CallFunction|PlayNewMontage))

(event LocalGameplayBehavior|EventOnAbort
  (Animation|Utilities|StopMontageifCurrent (Variables|Default|GetActivatingActor) (Variables|Default|GetPlayingMontage)))

(fn ConstructionScript)
